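# Niagara system for the idle effect on the ForestBath pickaxe, built from a single sprite emitter named Minimal.
# An instantaneous burst, limited by loop count, spawns cyan-tinted sprites with material randomness and a small camera offset.
# Particles ignore lifetime expiry and persist until the emitter deactivates, giving a steady sprite with no user parameters.
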NiagaraSystem /BRCosmetics/Weapons/FORT_Melee/Pickaxe_ForestBath/FX/NS_Pickaxe_ForestBath_Idle
dependencies {usedDataInterfaces [], usedDynamicInputs [], usedModules [&SystemState]}
emitters {Minimal {
  emitterUpdateScript [
   &EmitterState(Life Cycle Mode=::NewEnumerator1, Loop Behavior=::NewEnumerator1, Loop Duration Mode=::NewEnumerator1, UseLoopDelay=false)
   &SpawnBurst_Instantaneous(Use Spawn Probability=false, UseLoopCountLimit=true)
  ]
  particleSpawnScript [
   &InitializeParticle(
    Color=<rgba 0.1295 0.775 0.8906 1>
    Sprite Size=<36 36>
    Sprite Size Mode=::NewEnumerator4
    UsePositionOffset=false
    Write Material Random=true
   )
   &CameraOffset(Camera Offset Amount=2)
  ]
  particleUpdateScript [&ParticleState(Kill Particles When Lifetime Has Elapsed=false, Let Infinitely Lived Particles Die When Emitter Deactivates=true)]
  rendererClasses [&NiagaraSpriteRendererProperties]
  renderers [{rendererClass=&NiagaraSpriteRendererProperties}]
 }}
summary {
 systemName=NS_Pickaxe_ForestBath_Idle
 emitters [{emitterName=Minimal, rendererClasses [&NiagaraSpriteRendererProperties]}]
 userVariables []
}
