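# Physics asset for the front-end variant of the ForestBath two-handed pickaxe, setting up bodies and joints for its four dynamic rings.
# A capsule covers the handle and has no constraints.
# Each ring pairs a tiny box on its float bone with a thin, wide box on its secondary bone.
# One constraint per pair carries small swing and tighter twist limits, with no joints linking rings to each other or the handle.
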PhysicsAsset /BRCosmetics/Weapons/FORT_Melee/Pickaxe_ForestBath/Meshes/ForestBath_Axe_Physics_FrontEnd
bodies {
 Melee_TwoHanded_Handle [{
   shapeType=Capsule
   length=86.893
   radius=2.6052
   shapeName=Melee_TwoHanded_Handle_capsule
   center=<0.9024 -0.7431 25.3133>
   rotation=<rot 0.535 -90 0.9093>
  }]
 dyn_ring_a_float [{shapeType=Box, extentX=1.01, extentY=1.01, extentZ=1.01, shapeName=dyn_ring_a_float_box, rotation=<rot 0 -90 0>}]
 dyn_ring_a_secondary [{
   shapeType=Box
   extentX=2.4463
   extentY=18.1037
   extentZ=18.1017
   shapeName=dyn_ring_a_secondary_box
   center=<-0.1304 0.0045 -2.6726>
   rotation=<rot 0.3648 -90 -42.2743>
  }]
 dyn_ring_b_float [{shapeType=Box, extentX=1.01, extentY=1.01, extentZ=1.01, shapeName=dyn_ring_b_float_box, rotation=<rot 0 -90 0>}]
 dyn_ring_b_secondary [{
   shapeType=Box
   extentX=2.4013
   extentY=18.1034
   extentZ=18.1025
   shapeName=dyn_ring_b_secondary_box
   center=<-0.0076 -0.043 -3.1328>
   rotation=<rot -0.0946 -90 -2.9961>
  }]
 dyn_ring_c_float [{shapeType=Box, extentX=1.01, extentY=1.01, extentZ=1.01, shapeName=dyn_ring_c_float_box, rotation=<rot 0 -90 0>}]
 dyn_ring_c_secondary [{
   shapeType=Box
   extentX=2.4682
   extentY=18.1036
   extentZ=18.1012
   shapeName=dyn_ring_c_secondary_box
   center=<0.0456 0.0442 -3.2593>
   rotation=<rot 0.074 -90 -42.249>
  }]
 dyn_ring_d_float [{shapeType=Box, extentX=1.01, extentY=1.01, extentZ=1.01, shapeName=dyn_ring_d_float_box, rotation=<rot 0 -90 0>}]
 dyn_ring_d_secondary [{
   shapeType=Box
   extentX=2.5001
   extentY=18.103
   extentZ=18.1
   shapeName=dyn_ring_d_secondary_box
   center=<-0.1191 -0.0044 -3.1694>
   rotation=<rot 0.033 -90 -15.3749>
  }]
}
constraints [
 {bone1Name=dyn_ring_d_secondary, bone2Name=dyn_ring_d_float, swing1LimitDegrees=13, swing2LimitDegrees=13, twistLimitDegrees=5}
 {bone1Name=dyn_ring_b_secondary, bone2Name=dyn_ring_b_float, swing1LimitDegrees=13, swing2LimitDegrees=13, twistLimitDegrees=5}
 {bone1Name=dyn_ring_c_secondary, bone2Name=dyn_ring_c_float, swing1LimitDegrees=13, swing2LimitDegrees=13, twistLimitDegrees=5}
 {bone1Name=dyn_ring_a_secondary, bone2Name=dyn_ring_a_float, swing1LimitDegrees=13, swing2LimitDegrees=13, twistLimitDegrees=5}
]
