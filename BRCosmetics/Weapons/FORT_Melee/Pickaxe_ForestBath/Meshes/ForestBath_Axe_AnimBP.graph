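# Animation Blueprint for the ForestBath pickaxe, adding secondary motion to its mask, puffballs, tags and rings.
# A base pass offsets bones, springs the mask and trails the tag chain, then cached puffball/tag AnimDynamics variants feed the ring stage.
# bRigidBodyEnabled selects RigidBody simulation (regular or FrontEnd physics asset) or AnimDynamics rings (world or bone-relative).
# The event graph updates gravity override, sine-driven mask translation/rotation and an LOD2-or-lower flag.
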AnimBlueprint /BRCosmetics/Weapons/FORT_Melee/Pickaxe_ForestBath/Meshes/ForestBath_Axe_AnimBP
target_skeleton=&/ForestBath_Axe_Skeleton
graphs [
 AnimGraph {nodes [
   n1 = Anim Dynamics - Bone: dyn_ring_a_secondary!(ComponentPose=Local To Component(LocalPose=Use cached pose 'RbanOrNoRban'()), bAlphaBoolEnabled:bool=True, ExternalForce="0,0,0") -> (Pose) {
    alphaInputType=Bool
    angularDampingOverride=0.9999
    angularSpringConstant=1200
    bAngularSpring=true
    bOverrideAngularDamping=true
    bUseGravityOverride=true
    simulationSpace=BoneRelative
    alphaBoolBlend {blendOption=Linear}
    boundBone {boneName=dyn_ring_a_secondary}
    componentAppliedLinearAccClamp=<5000 5000 5000>
    componentLinearAccScale=<1 1 1>
    componentLinearVelScale=<1 1 1>
    gravityOverride {z=-980}
    physicsBodyDefinitions [{
      boundBone {boneName=dyn_ring_a_secondary}
      boxExtents=<20 3 20>
      constraintSetup {angularLimitsMax=<2 90 2>, angularLimitsMin=<-2 -90 -2>, angularTarget=<1 0 0>}
      localJointOffset=<0 0 -2.6>
     }]
    relativeSpaceBone {boneName=dyn_ring_a_float}
   }
   n2 = Anim Dynamics - Bone: dyn_ring_b_secondary!(ComponentPose=@n1, bAlphaBoolEnabled:bool=True, ExternalForce="0,0,0") -> (Pose) {
    alphaInputType=Bool
    angularDampingOverride=0.9999
    angularSpringConstant=1200
    bAngularSpring=true
    bOverrideAngularDamping=true
    bUseGravityOverride=true
    simulationSpace=BoneRelative
    alphaBoolBlend {blendOption=Linear}
    boundBone {boneName=dyn_ring_b_secondary}
    componentAppliedLinearAccClamp=<5000 5000 5000>
    componentLinearAccScale=<1 1 1>
    componentLinearVelScale=<1 1 1>
    gravityOverride {z=-980}
    physicsBodyDefinitions [{
      boundBone {boneName=dyn_ring_b_secondary}
      boxExtents=<15 3 15>
      constraintSetup {angularLimitsMax=<2 90 2>, angularLimitsMin=<-2 -90 -2>, angularTarget=<1 0 0>}
      localJointOffset=<0 0 -2.6>
     }]
    relativeSpaceBone {boneName=dyn_ring_b_float}
   }
   n3 = Anim Dynamics - Bone: dyn_ring_c_secondary!(ComponentPose=@n2, bAlphaBoolEnabled:bool=True, ExternalForce="0,0,0") -> (Pose) {
    alphaInputType=Bool
    angularDampingOverride=0.9999
    angularSpringConstant=1200
    bAngularSpring=true
    bOverrideAngularDamping=true
    bUseGravityOverride=true
    simulationSpace=BoneRelative
    alphaBoolBlend {blendOption=Linear}
    boundBone {boneName=dyn_ring_c_secondary}
    componentAppliedLinearAccClamp=<5000 5000 5000>
    componentLinearAccScale=<1 1 1>
    componentLinearVelScale=<1 1 1>
    gravityOverride {z=-980}
    physicsBodyDefinitions [{
      boundBone {boneName=dyn_ring_c_secondary}
      boxExtents=<20 3 20>
      constraintSetup {angularLimitsMax=<2 90 2>, angularLimitsMin=<-2 -90 -2>, angularTarget=<1 0 0>}
      localJointOffset=<0 0 -2.6>
     }]
    relativeSpaceBone {boneName=dyn_ring_c_float}
   }
   n4 = Anim Dynamics - Bone: dyn_ring_d_secondary!(ComponentPose=@n3, bAlphaBoolEnabled:bool=True, ExternalForce="0,0,0") -> (Pose) {
    alphaInputType=Bool
    angularDampingOverride=0.9999
    angularSpringConstant=1200
    bAngularSpring=true
    bOverrideAngularDamping=true
    bUseGravityOverride=true
    simulationSpace=BoneRelative
    alphaBoolBlend {blendOption=Linear}
    boundBone {boneName=dyn_ring_d_secondary}
    componentAppliedLinearAccClamp=<5000 5000 5000>
    componentLinearAccScale=<1 1 1>
    componentLinearVelScale=<1 1 1>
    gravityOverride {z=-980}
    physicsBodyDefinitions [{
      boundBone {boneName=dyn_ring_d_secondary}
      boxExtents=<15 3 15>
      constraintSetup {angularLimitsMax=<2 90 2>, angularLimitsMin=<-2 -90 -2>, angularTarget=<1 0 0>}
      localJointOffset=<0 0 -2.6>
     }]
    relativeSpaceBone {boneName=dyn_ring_d_float}
   }
   n5 = Component To Local(ComponentPose=@n4)
   n6 = Anim Dynamics - Bone: dyn_ring_a_secondary(ComponentPose=Local To Component(LocalPose=Use cached pose 'RbanOrNoRban'()), bAlphaBoolEnabled=True, ExternalForce="0,0,0") {
    alphaInputType=Bool
    angularDampingOverride=1
    angularSpringConstant=3000
    bAngularSpring=true
    bOverrideAngularDamping=true
    bUseGravityOverride=true
    simulationSpace=World
    alphaBoolBlend {blendOption=Linear}
    boundBone {boneName=dyn_ring_a_secondary}
    componentAppliedLinearAccClamp=<2000 2000 2000>
    componentLinearAccScale=<0.2 0.2 0.2>
    componentLinearVelScale=<0.2 0.2 0.2>
    gravityOverride {z=-980}
    physicsBodyDefinitions [{
      boundBone {boneName=dyn_ring_a_secondary}
      boxExtents=<22 22 22>
      constraintSetup {angularLimitsMax=<2 20 2>, angularLimitsMin=<-2 -20 -2>, angularTarget=<1 0 0>}
      localJointOffset=<1 0 0>
     }]
    relativeSpaceBone {boneName=dyn_ring_a_float}
   }
   n7 = Anim Dynamics - Bone: dyn_ring_b_secondary(ComponentPose=@n6, bAlphaBoolEnabled=True, ExternalForce="0,0,0") {
    alphaInputType=Bool
    angularDampingOverride=1
    angularSpringConstant=3000
    bAngularSpring=true
    bOverrideAngularDamping=true
    bUseGravityOverride=true
    simulationSpace=World
    alphaBoolBlend {blendOption=Linear}
    boundBone {boneName=dyn_ring_b_secondary}
    componentAppliedLinearAccClamp=<2000 2000 2000>
    componentLinearAccScale=<0.2 0.2 0.2>
    componentLinearVelScale=<0.2 0.2 0.2>
    gravityOverride {z=-980}
    physicsBodyDefinitions [{
      boundBone {boneName=dyn_ring_b_secondary}
      boxExtents=<22 22 22>
      constraintSetup {angularLimitsMax=<2 20 2>, angularLimitsMin=<-2 -20 -2>, angularTarget=<1 0 0>}
      localJointOffset=<1 0 0>
     }]
    relativeSpaceBone {boneName=dyn_ring_b_float}
   }
   n8 = Anim Dynamics - Bone: dyn_ring_c_secondary(ComponentPose=@n7, bAlphaBoolEnabled=True, ExternalForce="0,0,0") {
    alphaInputType=Bool
    angularDampingOverride=1
    angularSpringConstant=3000
    bAngularSpring=true
    bOverrideAngularDamping=true
    bUseGravityOverride=true
    simulationSpace=World
    alphaBoolBlend {blendOption=Linear}
    boundBone {boneName=dyn_ring_c_secondary}
    componentAppliedLinearAccClamp=<2000 2000 2000>
    componentLinearAccScale=<0.2 0.2 0.2>
    componentLinearVelScale=<0.2 0.2 0.2>
    gravityOverride {z=-980}
    physicsBodyDefinitions [{
      boundBone {boneName=dyn_ring_c_secondary}
      boxExtents=<22 22 22>
      constraintSetup {angularLimitsMax=<2 20 2>, angularLimitsMin=<-2 -20 -2>, angularTarget=<1 0 0>}
      localJointOffset=<1 0 0>
     }]
    relativeSpaceBone {boneName=dyn_ring_c_float}
   }
   n9 = Anim Dynamics - Bone: dyn_ring_d_secondary(ComponentPose=@n8, bAlphaBoolEnabled=True, ExternalForce="0,0,0") {
    alphaInputType=Bool
    angularDampingOverride=1
    angularSpringConstant=3000
    bAngularSpring=true
    bOverrideAngularDamping=true
    bUseGravityOverride=true
    simulationSpace=World
    alphaBoolBlend {blendOption=Linear}
    boundBone {boneName=dyn_ring_d_secondary}
    componentAppliedLinearAccClamp=<2000 2000 2000>
    componentLinearAccScale=<0.2 0.2 0.2>
    componentLinearVelScale=<0.2 0.2 0.2>
    gravityOverride {z=-980}
    physicsBodyDefinitions [{
      boundBone {boneName=dyn_ring_d_secondary}
      boxExtents=<22 22 22>
      constraintSetup {angularLimitsMax=<2 20 2>, angularLimitsMin=<-2 -20 -2>, angularTarget=<1 0 0>}
      localJointOffset=<1 0 0>
     }]
    relativeSpaceBone {boneName=dyn_ring_d_float}
   }
   n10 = Component To Local(ComponentPose=@n9)
   n11 = "Transform (Modify) Bone - Bone: dyn_ring_a_master"!(
    ComponentPose=Local To Component(LocalPose=Use cached pose 'RbanOrNoRban'())
    bAlphaBoolEnabled:bool=True
    Translation="0,0,1.3"
    Rotation="0,0,0"
    Scale="1,1,1"
   ) -> (Pose) {
    alphaInputType=Bool
    translationMode=BMM_Additive
    alphaBoolBlend {blendOption=Linear}
    boneToModify {boneName=dyn_ring_a_master}
    translation {z=1.3}
   }
   n12 = "Transform (Modify) Bone - Bone: dyn_ring_b_master"!(ComponentPose=@n11, bAlphaBoolEnabled:bool=True, Translation="0,0,1.3", Rotation="0,0,0", Scale="1,1,1") -> (Pose) {
    alphaInputType=Bool
    translationMode=BMM_Additive
    alphaBoolBlend {blendOption=Linear}
    boneToModify {boneName=dyn_ring_b_master}
    translation {z=1.3}
   }
   n13 = "Transform (Modify) Bone - Bone: dyn_ring_c_master"!(ComponentPose=@n12, bAlphaBoolEnabled:bool=True, Translation="0,0,1.3", Rotation="0,0,0", Scale="1,1,1") -> (Pose) {
    alphaInputType=Bool
    translationMode=BMM_Additive
    alphaBoolBlend {blendOption=Linear}
    boneToModify {boneName=dyn_ring_c_master}
    translation {z=1.3}
   }
   n14 = "Transform (Modify) Bone - Bone: dyn_ring_d_master"!(ComponentPose=@n13, bAlphaBoolEnabled:bool=True, Translation="0,0,1.3", Rotation="0,0,0", Scale="1,1,1") -> (Pose) {
    alphaInputType=Bool
    translationMode=BMM_Additive
    alphaBoolBlend {blendOption=Linear}
    boneToModify {boneName=dyn_ring_d_master}
    translation {z=1.3}
   }
   n15 = RigidBody!(ComponentPose=@n14, bAlphaBoolEnabled:bool, OverrideWorldGravity:Vector="0,0,0", ExternalForce) -> (Pose) {
    alphaInputType=Bool
    bClampLinearTranslationLimitToRefPose=true
    bForceDisableCollisionBetweenConstraintBodies=true
    bOverrideWorldGravity=true
    overridePhysicsAsset=&/ForestBath_Axe_Physics
    simulationSpace=BaseBoneSpace
    alphaBoolBlend {blendOption=Linear}
    baseBoneRef {boneName=Melee_TwoHanded_Handle}
    componentAppliedLinearAccClamp=<3000 3000 3000>
    componentLinearAccScale=<0.9 0.9 0.9>
    componentLinearVelScale=<0.9 0.9 0.9>
    simSpaceSettings {
     maxAngularAcceleration=500
     maxAngularVelocity=3
     maxLinearAcceleration=400
     maxLinearVelocity=300
     velocityScaleZ=0
     worldAlpha=1
     externalLinearDragV=<0.2 0.1 0.1>
    }
   }
   n16 = Component To Local(ComponentPose=@n15)
   n17 = RigidBody(ComponentPose=Local To Component(LocalPose=Use cached pose 'RbanOrNoRban'()), OverrideWorldGravity="0,0,0") {
    alphaInputType=Bool
    bClampLinearTranslationLimitToRefPose=true
    bForceDisableCollisionBetweenConstraintBodies=true
    bOverrideWorldGravity=true
    overridePhysicsAsset=&/ForestBath_Axe_Physics_FrontEnd
    simulationSpace=WorldSpace
    alphaBoolBlend {blendOption=Linear}
    baseBoneRef {boneName=Melee_TwoHanded_Handle}
    componentAppliedLinearAccClamp=<1200 1200 1200>
    componentLinearAccScale=<1 1 1>
    componentLinearVelScale=<1 1 1>
    simSpaceSettings {
     maxAngularAcceleration=500
     maxAngularVelocity=3
     maxLinearAcceleration=400
     maxLinearVelocity=300
     velocityScaleZ=0
     worldAlpha=1
     externalLinearDragV=<0.2 0.1 0.1>
    }
   }
   n18 = Component To Local(ComponentPose=@n17)
   Output Pose(Result=Blend Poses by bool(
     BlendPose_0=Blend Poses by bool(BlendPose_0=@n18, BlendPose_1=@n16) {blendType=Linear}
     BlendPose_1=Blend Poses by bool(BlendPose_0=@n10, BlendPose_1=@n5) {blendType=Linear}
     bActiveValue=Get bRigidBodyEnabled()
    ) {blendType=Linear}) {layerGroup=None}
   n19 = Anim Dynamics - Bone: dyn_puffballs_01!(ComponentPose=Local To Component(LocalPose=Use cached pose 'InitPose'()), bAlphaBoolEnabled:bool=True, ExternalForce="0,0,0") -> (Pose) {
    alphaInputType=Bool
    angularDampingOverride=0.9999
    angularSpringConstant=300
    bAngularSpring=true
    bOverrideAngularDamping=true
    bUseGravityOverride=true
    simulationSpace=World
    alphaBoolBlend {blendOption=Linear}
    boundBone {boneName=dyn_puffballs_01}
    componentAppliedLinearAccClamp=<4000 4000 4000>
    componentLinearAccScale=<1 1 1>
    componentLinearVelScale=<1 1 1>
    gravityOverride {z=-980}
    physicsBodyDefinitions [{
      boundBone {boneName=dyn_puffballs_01}
      boxExtents=<10 10 10>
      constraintSetup {angularLimitsMax=<1.5 45 2.5>, angularLimitsMin=<-1.5 -45 -2.5>, angularTarget=<1 0 0>}
      localJointOffset=<1 0 0>
     }]
    relativeSpaceBone {boneName=Melee_TwoHanded_Handle}
   }
   n20 = Anim Dynamics - Bone: dyn_puffballs_02!(ComponentPose=@n19, bAlphaBoolEnabled:bool=True, ExternalForce="0,0,0") -> (Pose) {
    alphaInputType=Bool
    angularDampingOverride=0.9999
    angularSpringConstant=300
    bAngularSpring=true
    bOverrideAngularDamping=true
    bUseGravityOverride=true
    simulationSpace=World
    alphaBoolBlend {blendOption=Linear}
    boundBone {boneName=dyn_puffballs_02}
    componentAppliedLinearAccClamp=<4000 4000 4000>
    componentLinearAccScale=<1 1 1>
    componentLinearVelScale=<1 1 1>
    gravityOverride {z=-980}
    physicsBodyDefinitions [{
      boundBone {boneName=dyn_puffballs_02}
      boxExtents=<10 10 10>
      constraintSetup {angularLimitsMax=<1.5 15 15>, angularLimitsMin=<-1.5 -15 -15>, angularTarget=<1 0 0>}
      localJointOffset=<1 0 0>
     }]
    relativeSpaceBone {boneName=Melee_TwoHanded_Handle}
   }
   n21 = Anim Dynamics - Bone: dyn_tag_01!(ComponentPose=@n20, bAlphaBoolEnabled:bool=True, ExternalForce="0,0,0") -> (Pose) {
    alphaInputType=Bool
    angularDampingOverride=0.9999
    angularSpringConstant=300
    bAngularSpring=true
    bOverrideAngularDamping=true
    bUseGravityOverride=true
    simulationSpace=World
    alphaBoolBlend {blendOption=Linear}
    boundBone {boneName=dyn_tag_01}
    componentAppliedLinearAccClamp=<4000 4000 4000>
    componentLinearAccScale=<1 1 1>
    componentLinearVelScale=<1 1 1>
    gravityOverride {z=-980}
    physicsBodyDefinitions [{
      boundBone {boneName=dyn_tag_01}
      boxExtents=<10 10 10>
      constraintSetup {angularLimitsMax=<1.5 15 15>, angularLimitsMin=<-1.5 -15 -15>, angularTarget=<1 0 0>}
      localJointOffset=<0.8 0 0>
     }]
    relativeSpaceBone {boneName=Melee_TwoHanded_Handle}
   }
   n22 = Anim Dynamics - Bone: dyn_tag_02!(ComponentPose=@n21, bAlphaBoolEnabled:bool=True, ExternalForce="0,0,0") -> (Pose) {
    alphaInputType=Bool
    angularDampingOverride=0.9999
    angularSpringConstant=300
    bAngularSpring=true
    bOverrideAngularDamping=true
    bUseGravityOverride=true
    simulationSpace=World
    alphaBoolBlend {blendOption=Linear}
    boundBone {boneName=dyn_tag_02}
    componentAppliedLinearAccClamp=<4000 4000 4000>
    componentLinearAccScale=<1 1 1>
    componentLinearVelScale=<1 1 1>
    gravityOverride {z=-980}
    physicsBodyDefinitions [{
      boundBone {boneName=dyn_tag_02}
      boxExtents=<10 10 10>
      constraintSetup {angularLimitsMax=<1.5 15 15>, angularLimitsMin=<-1.5 -15 -15>, angularTarget=<1 0 0>}
      localJointOffset=<0.8 0 0>
     }]
    relativeSpaceBone {boneName=Melee_TwoHanded_Handle}
   }
   n23 = Anim Dynamics - Bone: dyn_puffballs_01(ComponentPose=Local To Component(LocalPose=Use cached pose 'InitPose'()), bAlphaBoolEnabled=True, ExternalForce="0,0,0") {
    alphaInputType=Bool
    angularDampingOverride=0.999
    angularSpringConstant=100
    bAngularSpring=true
    bOverrideAngularDamping=true
    bUseGravityOverride=true
    simulationSpace=BoneRelative
    alphaBoolBlend {blendOption=Linear}
    boundBone {boneName=dyn_puffballs_01}
    componentAppliedLinearAccClamp=<5000 5000 5000>
    componentLinearAccScale=<1 1 1>
    componentLinearVelScale=<1 1 1>
    gravityOverride {z=-980}
    physicsBodyDefinitions [{
      boundBone {boneName=dyn_puffballs_01}
      boxExtents=<10 10 10>
      constraintSetup {angularLimitsMax=<1.5 45 2.5>, angularLimitsMin=<-1.5 -45 -2.5>, angularTarget=<1 0 0>}
      localJointOffset=<1 0 0>
     }]
    relativeSpaceBone {boneName=Melee_TwoHanded_Handle}
   }
   n24 = Anim Dynamics - Bone: dyn_puffballs_02(ComponentPose=@n23, bAlphaBoolEnabled=True, ExternalForce="0,0,0") {
    alphaInputType=Bool
    angularDampingOverride=0.999
    bOverrideAngularDamping=true
    bUseGravityOverride=true
    simulationSpace=BoneRelative
    alphaBoolBlend {blendOption=Linear}
    boundBone {boneName=dyn_puffballs_02}
    componentAppliedLinearAccClamp=<5000 5000 5000>
    componentLinearAccScale=<1 1 1>
    componentLinearVelScale=<1 1 1>
    gravityOverride {z=-980}
    physicsBodyDefinitions [{
      boundBone {boneName=dyn_puffballs_02}
      boxExtents=<10 10 10>
      constraintSetup {angularLimitsMax=<1.5 15 15>, angularLimitsMin=<-1.5 -15 -15>, angularTarget=<1 0 0>}
      localJointOffset=<1 0 0>
     }]
    relativeSpaceBone {boneName=Melee_TwoHanded_Handle}
   }
   n25 = Anim Dynamics - Bone: dyn_tag_01(ComponentPose=@n24, bAlphaBoolEnabled=True, ExternalForce="0,0,0") {
    alphaInputType=Bool
    angularDampingOverride=0.999
    bOverrideAngularDamping=true
    bUseGravityOverride=true
    simulationSpace=BoneRelative
    alphaBoolBlend {blendOption=Linear}
    boundBone {boneName=dyn_tag_01}
    componentAppliedLinearAccClamp=<5000 5000 5000>
    componentLinearAccScale=<1 1 1>
    componentLinearVelScale=<1 1 1>
    gravityOverride {z=-980}
    physicsBodyDefinitions [{
      boundBone {boneName=dyn_tag_01}
      boxExtents=<10 10 10>
      constraintSetup {angularLimitsMax=<1.5 15 15>, angularLimitsMin=<-1.5 -15 -15>, angularTarget=<1 0 0>}
      localJointOffset=<1 0 0>
     }]
    relativeSpaceBone {boneName=Melee_TwoHanded_Handle}
   }
   n26 = Anim Dynamics - Bone: dyn_tag_02(ComponentPose=@n25, bAlphaBoolEnabled=True, ExternalForce="0,0,0") {
    alphaInputType=Bool
    angularDampingOverride=0.999
    bOverrideAngularDamping=true
    bUseGravityOverride=true
    simulationSpace=BoneRelative
    alphaBoolBlend {blendOption=Linear}
    boundBone {boneName=dyn_tag_02}
    componentAppliedLinearAccClamp=<5000 5000 5000>
    componentLinearAccScale=<1 1 1>
    componentLinearVelScale=<1 1 1>
    gravityOverride {z=-980}
    physicsBodyDefinitions [{
      boundBone {boneName=dyn_tag_02}
      boxExtents=<10 10 10>
      constraintSetup {angularLimitsMax=<1.5 15 15>, angularLimitsMin=<-1.5 -15 -15>, angularTarget=<1 0 0>}
      localJointOffset=<1 0 0>
     }]
    relativeSpaceBone {boneName=Melee_TwoHanded_Handle}
   }
   Save cached pose 'RbanOrNoRban'(Pose=Blend Poses by bool(BlendPose_0=Component To Local(ComponentPose=@n22), BlendPose_1=Component To Local(ComponentPose=@n26)) {blendType=Linear}) {cacheName=RbanOrNoRban}
   AnimGraphNode_SubInput_0 = Input Pose()
   n27 = "Transform (Modify) Bone - Bone: dyn_puffballs_master"!(
    ComponentPose=Local To Component(LocalPose=@AnimGraphNode_SubInput_0)
    bAlphaBoolEnabled:bool=True
    Translation="0,0,0"
    Rotation="0,0,20"
    Scale="1,1,1"
   ) -> (Pose) {
    alphaInputType=Bool
    rotationMode=BMM_Additive
    scaleMode=BMM_Additive
    translationMode=BMM_Additive
    alphaBoolBlend {blendOption=Linear}
    boneToModify {boneName=dyn_puffballs_master}
    rotation {roll=20}
   }
   n28 = "Transform (Modify) Bone - Bone: dyn_mask_master"(ComponentPose=@n27, Alpha=1, Translation="0,-0.2,0", Rotation="0,0,0", Scale="0.95,0.95,0.95") {
    scaleMode=BMM_Replace
    translationMode=BMM_Additive
    alphaBoolBlend {blendOption=Linear}
    boneToModify {boneName=dyn_mask_master}
    scale=<0.95 0.95 0.95>
    translation {y=-0.2}
   }
   n29 = "Transform (Modify) Bone - Bone: dyn_mask_float"(ComponentPose=@n28, Alpha=1, Translation="0,0,0", Rotation="0,0,0", Scale="1,1,1") {
    rotationMode=BMM_Additive
    scaleMode=BMM_Replace
    translationMode=BMM_Additive
    alphaBoolBlend {blendOption=Linear}
    boneToModify {boneName=dyn_mask_float}
   }
   n30 = Anim Dynamics - Bone: dyn_mask_secondary!(ComponentPose=@n29, Alpha=1, GravityOverride:Vector="0,0,0", ExternalForce="0,0,0") -> (Pose) {
    angularDampingOverride=1
    angularSpringConstant=500
    bAngularSpring=true
    bGravityOverrideInSimSpace=true
    bLinearSpring=true
    bOverrideAngularDamping=true
    bOverrideLinearDamping=true
    bUseGravityOverride=true
    linearDampingOverride=1
    linearSpringConstant=300
    simulationSpace=BoneRelative
    alphaBoolBlend {blendOption=Linear}
    boundBone {boneName=dyn_mask_secondary}
    componentAppliedLinearAccClamp=<5000 5000 5000>
    componentLinearAccScale=<0.3 0.3 0.3>
    componentLinearVelScale=<0.3 0.3 0.3>
    physicsBodyDefinitions [{
      boundBone {boneName=dyn_mask_secondary}
      boxExtents=<10 10 10>
      constraintSetup {angularLimitsMax=<5 5 5>, angularLimitsMin=<-5 -5 -5>, angularTarget=<1 0 0>, linearAxesMax=<0.8 0 0.8>, linearAxesMin=<-0.8 0 -0.8>}
      localJointOffset=<0.2 0 0>
     }]
    relativeSpaceBone {boneName=dyn_mask_float}
   }
   n31 = Trail controller - Bone: dyn_tag_02!(ComponentPose=@n30, bAlphaBoolEnabled:bool=True, RelaxationSpeedScale:float=1) -> (Pose) {
    alphaInputType=Bool
    bActorSpaceFakeVel=true
    bLimitRotation=true
    bShowTrailLocation=true
    bUsePlanarLimit=true
    chainLength=4
    lastBoneRotationAnimAlphaBlend=1
    maxDeltaTime=0.0333
    alphaBoolBlend {blendOption=CubicInOut}
    alphaScaleBiasClamp {bInterpResult=true}
    baseJoint {boneName=dyn_puffballs_master}
    fakeVelocity {y=150, z=-150}
    relaxationSpeedScaleInputProcessor {bInterpResult=true, interpSpeedDecreasing=25, interpSpeedIncreasing=25}
    rotationLimits [
     {limitMax=<3 25 5>, limitMin=<-3 -25 -5>}
     {limitMax=<3 15 15>, limitMin=<-3 -15 -15>}
     {limitMax=<3 15 15>, limitMin=<-3 -15 -15>}
     {limitMax=<3 15 15>, limitMin=<-3 -15 -15>}
    ]
    rotationOffsets [<0 0 0>, <0 0 0>, <0 0 0>, <0 0 0>]
    trailBone {boneName=dyn_tag_02}
    trailRelaxationSpeed {editorCurveData {keys [
       <at 0 10 RCIM_Linear>
       {arriveTangent=-6, interpMode=RCIM_Cubic, leaveTangent=-6, time=0.6531, value=8.1902}
       {arriveTangent=-22.8779, interpMode=RCIM_Cubic, leaveTangent=-22.8779, tangentMode=RCTM_User, time=1, value=4}
      ]}}
   }
   Save cached pose 'InitPose'(Pose=Component To Local(ComponentPose=@n31)) {cacheName=InitPose}
  ]}
 EventGraph {nodes [
   Event BlueprintUpdateAnimation()
   > n1 = Sequence!(execute) -> (then_0, then_1, then_2:Exec, then_3:Exec)
   > n2 = Gravity Override State Function DEPRECATED(CharacterPartAnimInstance=Self-Reference(), GameTimeInSeconds=GetGameTimeInSeconds(), GravityOverrideParams=Get Gravity Override Params())
   > Set GravityOverride(GravityOverride=@n2)
   n3 = GetRealTimeSeconds()
   n4 = "Sin (Radians)"(A=float * float(A=@n3, B=float * float(A=0.8, B=-1.2)) {is=title K2Node_PromotableOperator})
   n5 = float * float(A=0.8, B=-0.8)
   n6 = float / float(A="Sin (Radians)"(A=float * float(A=@n5, B=@n3)), B=-0.7)
   Set Translate Mask(execute=@n1.then_1, Translate Mask:Vector=MakeVector(X=float + float(A=@n6, B=2), Y=0, Z=float / float(A=@n4, B=-0.7))) -> (Output_Get:Vector)
   n7 = GetRealTimeSeconds()
   n8 = "Sin (Radians)"(A=float * float(A=float * float(A=0.75, B=-0.8), B=@n7))
   n9 = "Sin (Radians)"(A=float * float(A=@n7, B=float * float(A=0.75, B=-1.5)))
   n10 = "Sin (Radians)"(A=float * float(A=@n7, B=float * float(A=0.75, B=-1.5)))
   Set Rotate Mask(
    execute=@n1.then_2
    Rotate Mask:Rotator=MakeRotator(Roll=float / float(A=@n8, B=-1), Pitch=float / float(A=@n10, B=-1), Yaw=float / float(A=@n9, B=-1))
   ) -> (Output_Get:Rotator)
   Set is LOD2 or Lower(execute=@n1.then_3, is LOD2 or Lower:bool="integer >= integer"(A=Get CurrentLOD(), B=2)) -> (Output_Get:bool)
   TryGetPawnOwner()
  ]}
]
variables [
 Gravity Override Params {type=GravityOverrideParamsStruct}
 GravityOverride {type=Vector}
 is LOD2 or Lower {type=boolean}
 Rotate Mask {type=Rotator}
 Translate Mask {type=Vector}
]
